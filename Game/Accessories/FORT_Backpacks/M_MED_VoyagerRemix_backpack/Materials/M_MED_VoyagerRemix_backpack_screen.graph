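# Animated screen material for the VoyagerRemix backpack, showing a pixelated space-shooter scene as emissive color.
# UVs are quantized to a pixel grid, then sample masks for scrolling stars, a sine-drifting spaceship and looping missiles, combined by max.
# An antialiased grid overlay uses screen-space derivatives; camera distance raises resolution and fades the grid to average intensity.
# A tinted box-mask backlight is added before the accessories override function.
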Material /Game/Accessories/FORT_Backpacks/M_MED_VoyagerRemix_backpack/Materials/M_MED_VoyagerRemix_backpack_screen
graph {
 properties {
  bUseMaterialAttributes=true
  bUsedWithMeshDeformer=true
  bUsedWithSkeletalMesh=true
  pixelDepthOffsetMode=PDOM_Legacy
  usedShadingModels=Default Lit
 }
 nodes [
  n1 = Subtract(A=ActorPositionWS(), B=CameraPositionWS())
  n2 = Saturate(None=MaterialFunctionCall(
    Input=MaterialFunctionCall(Vector 3=@n1) {materialFunction=/Engine/Functions/Engine_MaterialFunctions02/Utility/VectorLength}.V3 Length
    Input Low=ScalarParameter() {defaultValue=2000, group=DistanceFade, parameterName=QuantizationFadeStart}
    Input High=ScalarParameter() {defaultValue=4000, group=DistanceFade, parameterName=QuantizationFadeEnd}
    Target Low=Constant()
    Target High=Constant() {r=1}
   ) {materialFunction=&RemapValueRange}.Result)
  n3 = Multiply(A=Add(A=Multiply(A=@n2) {constB=2}), B=ScalarParameter() {defaultValue=70, group=Display, parameterName=ScreenResolution})
  MaterialExpressionMultiply_1 = Multiply(A=ScalarParameter() {defaultValue=7, group=Display, parameterName=PixelPitch}, B=Divide(B=@n3)) {constB=7}
  n4 = TextureCoordinate()
  n5 = Frac(None=Multiply(A=@n4, B=@n3))
  n6 = Min(A=@n5, B=OneMinus(None=@n5))
  n7 = Add(A=Abs(None=DDX(Value=@n6)), B=Abs(None=DDY(Value=@n6)))
  MaterialExpressionSubtract_9 = Subtract(A=@n6, B=@n7)
  MaterialExpressionAdd_19 = Add(A=@n6, B=@n7)
  n8 = Custom!(varA=@MaterialExpressionMultiply_1, varMin=@MaterialExpressionSubtract_9, varMax=@MaterialExpressionAdd_19) {
   code="return smoothstep(varMin,varMax,varA);"
   description=smoothStep_V2
   outputType=CMOT_Float2
   inputs [
    {
     inputName=varA
     input {expression=&self:MaterialExpressionMultiply_1}
    }
    {
     inputName=varMin
     input {expression=&self:MaterialExpressionSubtract_9}
    }
    {
     inputName=varMax
     input {expression=&self:MaterialExpressionAdd_19}
    }
   ]
  }
  n9 = MaterialFunctionCall(Float2=@n8) {materialFunction=&BreakOutFloat2Components}
  n10 = Saturate(None=Add(A=@n9.R, B=@n9.R))
  n11 = Constant()
  n12 = ScalarParameter() {defaultValue=-0.2, group=Anim, parameterName=StarsSpeed}
  n13 = Divide(A=Floor(None=Multiply(A=@n4, B=@n3) {constB=5}), B=@n3) {constB=5}
  n14 = Multiply(A=Panner(Coordinate=@n13, Time=Time(), Speed=AppendVector(A=@n11, B=@n12)), B=Constant() {r=4}) {constB=5}
  n15 = TextureObject() {samplerType=SAMPLERTYPE_Masks, texture=/Game/Accessories/FORT_Backpacks/M_MED_VoyagerRemix_backpack/Textures/T_VoyagerRemixScreen}
  n16 = TextureSample(UVs=Frac(None=Add(A=@n14) {constB=0}), Tex=@n15) {
   constMipValue=-3
   samplerSource=SSM_Wrap_WorldGroupSettings
   samplerType=SAMPLERTYPE_Masks
   texture=/Game/Accessories/FORT_Backpacks/M_MED_VoyagerRemix_backpack/Textures/T_M_MED_VoyagerRemix_backpack_E
  }
  n17 = Multiply(
   A=Power(Base=@n16.B, Exp=ScalarParameter() {defaultValue=1, group=Colors, parameterName=StarsPow})
   B=ScalarParameter() {defaultValue=0.4, group=Colors, parameterName=StarsIntensity}
  )
  n18 = Time()
  n19 = Add(A=@n18) {constB=0}
  n20 = Multiply(
   A=AppendVector(A=Multiply(A=@n18) {constB=0.3}, B=Multiply(A=@n19) {constB=0.2})
   B=ScalarParameter() {defaultValue=2, group=Anim, parameterName=SpaceShip Speed}
  )
  n21 = Add(A=@n13, B=Multiply(A=Sine(None=@n20), B=ScalarParameter() {defaultValue=0.1, group=Anim, parameterName=SpaceShip Range}))
  n22 = Constant() {r=6}
  n23 = Constant2Vector() {g=-1.8, r=-2.5}
  n24 = TextureSample(UVs=Add(A=Multiply(A=@n21, B=@n22) {constB=6}, B=@n23) {constB=0}, Tex=@n15) {
   constMipValue=-3
   samplerSource=SSM_Clamp_WorldGroupSettings
   samplerType=SAMPLERTYPE_Masks
   texture=&T_M_MED_VoyagerRemix_backpack_E
  }
  n25 = Multiply(
   A=Power(Base=@n24.G, Exp=ScalarParameter() {defaultValue=1, group=Colors, parameterName=SpaceShipPow})
   B=ScalarParameter() {defaultValue=1, group=Colors, parameterName=SpaceShipIntensity}
  )
  n26 = Panner(
   Coordinate=@n21
   Time=Frac(None=Time())
   Speed=AppendVector(A=Constant(), B=ScalarParameter() {defaultValue=0.7, group=Anim, parameterName=MissilesSpeed})
  )
  n27 = TextureSample(UVs=Add(A=Multiply(A=@n26, B=@n22) {constB=6}, B=@n23) {constB=0}, Tex=@n15) {
   samplerSource=SSM_Clamp_WorldGroupSettings
   samplerType=SAMPLERTYPE_Masks
   texture=&T_M_MED_VoyagerRemix_backpack_E
  }
  n28 = Power(Base=@n27.R, Exp=ScalarParameter() {defaultValue=1.5, group=Colors, parameterName=MissilesPow})
  n29 = ScalarParameter() {defaultValue=1, group=Colors, parameterName=MissilesIntensity}
  n30 = Max(A=Max(A=Saturate(None=@n17), B=Saturate(None=@n25)) {constB=0}, B=Saturate(None=Multiply(A=@n28, B=@n29)))
  n31 = ScalarParameter() {defaultValue=0.003, group=Display, parameterName=ScreenMinIntensity}
  n32 = VectorParameter() {group=Display, parameterName=ScreenColor, defaultValue=<rgba 3 3 3 0>}
  n33 = Multiply(
   A=LinearInterpolate(A=OneMinus(None=@n10), B=ScalarParameter() {defaultValue=0.4, group=Display, parameterName=AverageGridIntensity}, Alpha=@n2)
   B=Multiply(A=Max(A=@n30, B=@n31), B=@n32.RGB)
  ) {constA=1}
  n34 = Constant3Vector() {constant=<rgba 0.5 0.35 0 0>}
  n35 = Constant()
  n36 = ScalarParameter() {defaultValue=1.5, group=Display, parameterName=BackLightFalloff}
  n37 = Multiply(
   A=OneMinus(None=MaterialFunctionCall!(A=@n4, B=@n34, Bounds=@n35, Edge Falloff=@n36) {materialFunction=/Engine/Functions/Engine_MaterialFunctions01/Gradient/BoxMask-2D}.Result)
   B=ScalarParameter() {defaultValue=2.2, group=Display, parameterName=BackLightIntensity}
  ) {constB=2}
  n38 = ScalarParameter() {defaultValue=3.5, group=Display, parameterName=BackLightPow}
  n39 = Multiply(A=@n32.RGB, B=Saturate(None=Power(Base=@n37, Exp=@n38) {constExponent=1}))
  MaterialFunctionCall(In=SetMaterialAttributes(Emissive Color=Add(A=@n33, B=@n39)) {attributeSetTypes [#1]}) {materialFunction=&MF_TechArtOverrides_accessories}
 ]
}
